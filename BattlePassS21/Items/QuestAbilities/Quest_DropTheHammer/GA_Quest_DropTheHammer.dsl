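; Gameplay ability for the Drop The Hammer quest that detects when the player's count of tagged items falls and reports it as quest progress.
; An event whose target pickup matches the tag increments a stored count and, through a DoOnce, starts a looping one-second timer; non-matching events end the ability.
; Each tick recounts matching inventory items, with debug prints, then either saves the count or, on a decrease, clears the timer and sends the quest stat event.
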(event Ability|EventActivateAbilityFromEvent
  (params (EventData :type Gameplay%20Event%20Data%20Structure%20%28by%20ref%29))
  (bind (_eventtag _instigator _target) (Utilities|Struct|BreakGameplayEventData EventData))
  (|CallParentFunction EventData)
  (if (GameplayTags|HasMatchingGameplayTag (Pickup|GetPrimaryItemDefinition (Utilities|Casting|CastToFortPickupAthena _target)) (Variables|Default|GetTag))
    (Math|Float|IncrementFloat (Variables|Default|GetHammer%20Count_Original)
      (:%20%20
        (Utilities|FlowControl|DoOnce false
          (:Completed
            (bind _returnvalue (Utilities|Time|SetTimerbyEvent (delegate Custom|CustomEvent) 1.0 true))
            (Variables|Default|SetTimer%20Handle _returnvalue)))))
    (else
      (Ability|EndAbility))))

(event Custom|CustomEvent
  (Variables|Default|SetHammer%20Count%20New 0.0)
  (bind _items (InventoryManagement|GetItemsfromInventoryOwner (Fortnite|GetFortPlayerControllerfromActor (Variables|GAVars|Data|GetOwningPlayerPawn))))
  (for _array_element _items
    (if (GameplayTags|HasMatchingGameplayTag (Item|GetItemDefinition _array_element) (Variables|Default|GetTag))
      (bind _result (Math|Float|IncrementFloat (Variables|Default|GetHammer%20Count%20New))
        (:%20%20
          (Development|PrintString (Utilities|String|ToString(Float) _result) true true "(R=1.000000,G=0.000000,B=0.024272,A=1.000000)" 1.0)))))
  (if (> (Variables|Default|GetHammer%20Count_Original) (Variables|Default|GetHammer%20Count%20New))
    (Utilities|Time|ClearandInvalidateTimerbyHandle (Variables|Default|GetTimer%20Handle))
    (Quest|SendComplexCustomStatEvent (Variables|GAVars|Data|GetOwningPlayerQuestManager) 0 "(GameplayTags=((TagName=\"Athena.Quests.DropTheHammer\")))" "(GameplayTags=)" 1)
    (else
      (Variables|Default|SetHammer%20Count_Original (Variables|Default|GetHammer%20Count%20New)))))
